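; Passive Battle Pass quest ability for surviving storm circles while carrying a qualifying weapon (VR+ rarity, per the asset name).
; On each storm zone update, it gathers the owning player's inventory items across the required rarities.
; Any item with a required weapon tag and no excluded tag switches the quest source tag to the VRPlusWeapon variant.
; The parent handler then runs with that tag, after which the tag resets to plain SurviveStormCircles.
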(event Ability|EventFailedToActivatePassiveAbility
  (params (ActorInfo :type Gameplay%20Ability%20Actor%20Info%20Structure)))

(event Ability|EventActivateAbility
  (|CallParentFunction))

(event EventStormZoneUpdated
  (params (SafeZonePhase :type Integer))
  ;[1 1 1] When storm zone updates, check if player has a weapon of any of the specified rarity types in inventory
  (Utilities|Array|Clear (Variables|Default|GetPlayerItemsOfReqRarities))
  (Variables|Default|SetOwningPlayerInventoryInterface (Fortnite|GetFortPlayerControllerfromActor (Variables|GAVars|Data|GetOwningPlayerPawn)))
  (Utilities|FlowControl|Sequence
    (:then_0
      ;;[1 1 1] Get all items of the specified rarities and combine to one array
      (for _array_element (Variables|Default|GetRequiredRarityTypes)
        (bind _items (InventoryManagement|GetItemsfromInventoryOwnerforRarity (Variables|Default|GetOwningPlayerInventoryInterface) _array_element))
        (if (Utilities|Array|IsNotEmpty _items)
          (Utilities|Array|AppendArray (Variables|Default|GetPlayerItemsOfReqRarities) _items))))
    (:then_1
      (bind _array_element_2 (Utilities|Array|ForEachLoopwithBreak (Variables|Default|GetPlayerItemsOfReqRarities))
        (:LoopBody
          (if (and (GameplayTags|HasAnyTags (CallFunction|GetOwnedGameplayTags(Fix) _array_element_2) (Variables|Default|GetT_RequiredWeaponTags)) (not (GameplayTags|HasAnyTags (CallFunction|GetOwnedGameplayTags(Fix) _array_element_2) (Variables|Default|GetT_ExcludedWeaponTags) true)))
            (Variables|Default|SetQuestSourceTag "(GameplayTags=((TagName=\"Athena.Quests.SurviveStormCircles.VRPlusWeapon\")))")))
        (:Completed
          _))
      (|CallParentFunction 0)
      (Variables|Default|SetQuestSourceTag "(GameplayTags=((TagName=\"Athena.Quests.SurviveStormCircles\")))"))))

(fn GetOwnedGameplayTags(Fix)
  (params (Item :type Fort%20World%20Item%20Object%20Reference))
  (return (GameplayTags|GetOwnedGameplayTags Item :!class /Script/FortniteGame.FortWorldItem)))
